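# Niagara effect for secret base vent wind: three pairs of red-tinted, angled air jets spaced around a ring.
# All three emitters share placement math, using partition index and spawn-group masks to derive each jet's position and tilted axis.
# Heat sprites and GPU Motes launch along that axis, accelerate upward and fade red to white; Heat feeds LOD distance into its material.
# Side_Wind holds six persistent meshes with waveform-pulsed length and age-driven material animation.
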NiagaraSystem /Hera/Assets/FX/wind/NS_SecretBaseVentWind
dependencies {
 usedDataInterfaces [&NiagaraDataInterfaceCurve, &NiagaraDataInterfaceColorCurve]
 usedDynamicInputs []
 usedModules [&SystemState]
}
emitters {
 Heat {
  emitterUpdateScript [
   &EmitterState()
   &SpawnRate(SpawnRate=10, Use Spawn Probability=false)
   &SpawnRate(SpawnGroup=1, SpawnRate=10, Use Spawn Probability=false) {module=SpawnRate001}
  ]
  particleSpawnScript [
   &InitializeParticle(
    Color Mode=::NewEnumerator0
    Lifetime Max=1.8
    Lifetime Min=2
    Lifetime Mode=::NewEnumerator1
    Mass Mode=::NewEnumerator3
    Sprite Rotation Mode=::NewEnumerator2
    Sprite Size Mode=::NewEnumerator3
    Uniform Sprite Size=650
    UsePositionOffset=false
    Write Material Random=true
   )
   /SetVariables(Particles.Spread=0.0325) {isSetParameters=true, module=SetVariables_#1}
   &PartitionParticles(
    Particles Per Partition=3
    Partition Count Mode:ENiagara_InfiniteLoopDuration=::NewEnumerator1
    Partition Mode:ENiagara_PartitionMode_v2=::NewEnumerator1
    Write Partition Index to Ribbon Link Order:ENiagara_AttributeSamplingApplyOutput=::NewEnumerator0
    Write Partition to Ribbon ID=::NewEnumerator0
   )
   &ShapeLocation(
    MaskSpawnGroup=false
    Offset=<0 0 -275>
    Offset Mode=::NewEnumerator0
    Override Seed=false
    Ring / Disc Distribution Mode:ENiagaraSphereDistributionMode=::NewEnumerator1
    Ring Radius=3000
    Shape Primitive=::NewEnumerator4
    U Position=&Add_Float(
     A=0.25
     B=&MaskFloatBySpawnGroup(
      Masked Float=&Add_Float(A=$Particles.PartitionParticles.Normalized Index In Partition, B=$Particles.Spread)
      Passthrough Float=&Subtract_Float(A=$Particles.PartitionParticles.Normalized Index In Partition, B=$Particles.Spread)
     )
    )
    UseOwnerScale=true
   )
   &InitialMeshOrientation(
    Mesh Orientation Mode=::NewEnumerator2
    Rotation=&MaskVectorBySpawnGroup(
     Masked Vector=&MakeVector(X=0, Y=0, Z=&Add_Float(A=0.055, B=$Particles.PartitionParticles.Normalized Index In Partition))
     Passthrough Vector=&MakeVector(X=0, Y=0, Z=&Add_Float(A=-0.055, B=$Particles.PartitionParticles.Normalized Index In Partition))
    )
    Rotation Coordinate Space=::NewEnumerator0
    Use Rotation Vector=true
   )
   /SetVariables(Particles.MeshOrientation=&MultiplyQuaternion(
     Quaternion A=&MakeQuaternion(
      Quaternion From=0
      Axis=&ReturnMeshOrientationAxis(Mesh Orientation Axis=::XAxis, Mesh Orientation Quat=$Particles.MeshOrientation)
      Angle=47
      From=?
      To=?
      Angle Type=::NewEnumerator0
      Normalize Vectors=false
     )
     Quaternion B=$Particles.MeshOrientation
    )) {isSetParameters=true, module=SetVariables_#2}
   /SetVariables(Particles.Tube Axis:Vector3f=&ReturnMeshOrientationAxis(Mesh Orientation Axis=::YAxis, Mesh Orientation Quat=$Particles.MeshOrientation)) {isSetParameters=true, module=SetVariables_#3}
   &AddVelocity(Velocity=$Particles.Tube Axis, Velocity Speed Scale=500)
   &CameraOffset(Camera Offset Amount=200)
  ]
  particleUpdateScript [
   &ParticleState()
   &SpriteRotationRate(Rotation Rate=&Multiply_Float(A=30, B=&RandomRangeFloat(Minimum=-1, Maximum=1)))
   &AccelerationForce(Acceleration=<0 0 980>)
   &Drag(Drag=1.5, Use Linear Drag=true, Use Rotational Drag=true)
   &SolveForcesAndVelocity(Clamp Velocity=false, Limit Acceleration=false)
   &DynamicMaterialParameters(
    Index 3 Param 4=&FloatFromCurve(
     FloatCurve=DataInterface() {propertyValues {Curve {keys [
         <at 0 1 RCIM_Cubic>
         <at 0.1388 1.15>
         {arriveTangent=-1.2839, interpMode=RCIM_Cubic, leaveTangent=-1.2839, tangentMode=RCTM_User, time=1, value=1}
        ]}}}
     CurveIndex=&FN_LOD_Distance_Remap(Start Distance=0, End Distance=15000, Minimum Output=0, Maximum Output=1, Clamp Output=true)
    )
    Module.Index3Float0Write=false
    Module.Index3Float1Write=false
    Module.Index3Float2Write=false
    Module.Index3Float3Write=true
    Write Parameter Index 0=::NewEnumerator0
    Write Parameter Index 3=::NewEnumerator1
   )
   &Color(
    Color=&ColorFromCurve(ColorCurve=DataInterface() {propertyValues {
       AlphaCurve {
        defaultValue=1
        keys [
         {arriveTangent=9.0725, interpMode=RCIM_Cubic, leaveTangent=9.0725, tangentMode=RCTM_User, time=0, value=0}
         {
          arriveTangent=-2.5249
          arriveTangentWeight=0.0469
          interpMode=RCIM_Cubic
          leaveTangent=-2.5249
          tangentMode=RCTM_User
          time=0.1408
          value=0.9847
         }
         {arriveTangent=-0.0405, arriveTangentWeight=0.2864, interpMode=RCIM_Cubic, leaveTangent=-0.0405, tangentMode=RCTM_User, time=1, value=0}
        ]
       }
       BlueCurve {keys [<at 0.2333 0>, {arriveTangent=1.0075, interpMode=RCIM_Cubic, time=0.6718, value=1}, <at 1 1 RCIM_Cubic>]}
       GreenCurve {keys [<at 0.2333 0>, {arriveTangent=1.0075, interpMode=RCIM_Cubic, time=0.6718, value=1}, <at 1 1 RCIM_Cubic>]}
       RedCurve {keys [<at 0.2333 1>,<at 0.6718 1 RCIM_Cubic>,<at 1 1 RCIM_Cubic>]}
      }})
    Scale Alpha=0.065
   )
  ]
  rendererClasses [&NiagaraSpriteRendererProperties]
  renderers [{rendererClass=&NiagaraSpriteRendererProperties}]
 }
 Motes {
  simTarget=GPUComputeSim
  emitterSpawnScript [/SetVariables(Emitter.rate=100) {isSetParameters=true}]
  emitterUpdateScript [
   &EmitterState()
   &SpawnRate(SpawnRate=$Emitter.rate, Use Spawn Probability=false)
   &SpawnRate(SpawnGroup=1, SpawnRate=$Emitter.rate, Use Spawn Probability=false) {module=SpawnRate001}
  ]
  particleSpawnScript [
   &InitializeParticle(
    Color Mode=::NewEnumerator0
    Lifetime Max=1.35
    Lifetime Mode=::NewEnumerator1
    Mass Mode=::NewEnumerator3
    Sprite Rotation Mode=::NewEnumerator2
    Sprite Size Max=<5 10>
    Sprite Size Min=<5 50>
    Sprite Size Mode=::NewEnumerator2
    UsePositionOffset=false
    Write Material Random=true
   )
   /SetVariables(Particles.Spread=0.0325) {isSetParameters=true, module=SetVariables_#1}
   &PartitionParticles(
    Particles Per Partition=3
    Partition Count Mode:ENiagara_InfiniteLoopDuration=::NewEnumerator1
    Partition Mode:ENiagara_PartitionMode_v2=::NewEnumerator1
    Write Partition Index to Ribbon Link Order:ENiagara_AttributeSamplingApplyOutput=::NewEnumerator0
    Write Partition to Ribbon ID=::NewEnumerator0
   )
   &ShapeLocation(
    MaskSpawnGroup=false
    Offset=<0 0 -275>
    Offset Mode=::NewEnumerator0
    Override Seed=false
    Ring / Disc Distribution Mode:ENiagaraSphereDistributionMode=::NewEnumerator1
    Ring Radius=3000
    Shape Primitive=::NewEnumerator4
    U Position=&Add_Float(
     A=0.25
     B=&MaskFloatBySpawnGroup(
      Masked Float=&Add_Float(A=$Particles.PartitionParticles.Normalized Index In Partition, B=$Particles.Spread)
      Passthrough Float=&Subtract_Float(A=$Particles.PartitionParticles.Normalized Index In Partition, B=$Particles.Spread)
     )
    )
    UseOwnerScale=true
   )
   &InitialMeshOrientation(
    Mesh Orientation Mode=::NewEnumerator2
    Rotation=&MaskVectorBySpawnGroup(
     Masked Vector=&MakeVector(X=0, Y=0, Z=&Add_Float(A=0.055, B=$Particles.PartitionParticles.Normalized Index In Partition))
     Passthrough Vector=&MakeVector(X=0, Y=0, Z=&Add_Float(A=-0.055, B=$Particles.PartitionParticles.Normalized Index In Partition))
    )
    Rotation Coordinate Space=::NewEnumerator0
    Use Rotation Vector=true
   )
   /SetVariables(Particles.MeshOrientation=&MultiplyQuaternion(
     Quaternion A=&MakeQuaternion(
      Quaternion From=0
      Axis=&ReturnMeshOrientationAxis(Mesh Orientation Axis=::XAxis, Mesh Orientation Quat=$Particles.MeshOrientation)
      Angle=47
      From=?
      To=?
      Angle Type=::NewEnumerator0
      Normalize Vectors=false
     )
     Quaternion B=$Particles.MeshOrientation
    )) {isSetParameters=true, module=SetVariables_#2}
   /SetVariables(Particles.Tube Axis:Vector3f=&ReturnMeshOrientationAxis(Mesh Orientation Axis=::YAxis, Mesh Orientation Quat=$Particles.MeshOrientation)) {isSetParameters=true, module=SetVariables_#3}
   &AddVelocity(Velocity=$Particles.Tube Axis, Velocity Speed Scale=500)
   &CameraOffset(Camera Offset Amount=200)
   &ShapeLocation(
    Disc Coverage=1
    MaskSpawnGroup=false
    Override Seed=false
    Ring Radius=250
    Rotation Axis=$Particles.Tube Axis
    Rotation Mode=::NewEnumerator1
    Shape Primitive=::NewEnumerator4
    UseOwnerScale=true
   ) {module=ShapeLocation001}
  ]
  particleUpdateScript [
   &ParticleState()
   &SpriteRotationRate(Rotation Rate=&Multiply_Float(A=30, B=&RandomRangeFloat(Minimum=-1, Maximum=1)))
   &AccelerationForce(Acceleration=&Multiply_VectorByFloat(Vector=<0 0 980>, Float=0.5))
   &Drag(Drag=1.5, Use Linear Drag=true, Use Rotational Drag=true)
   &SolveForcesAndVelocity(Clamp Velocity=false, Limit Acceleration=false)
   &Color(
    Color=&ColorFromCurve(ColorCurve=DataInterface() {propertyValues {
       AlphaCurve {
        defaultValue=1
        keys [
         {arriveTangent=9.0725, interpMode=RCIM_Cubic, leaveTangent=9.0725, tangentMode=RCTM_User, time=0, value=0}
         {
          arriveTangent=-2.5249
          arriveTangentWeight=0.0469
          interpMode=RCIM_Cubic
          leaveTangent=-2.5249
          tangentMode=RCTM_User
          time=0.1408
          value=0.9847
         }
         {arriveTangent=-0.0405, arriveTangentWeight=0.2864, interpMode=RCIM_Cubic, leaveTangent=-0.0405, tangentMode=RCTM_User, time=1, value=0}
        ]
       }
       BlueCurve {keys [<at 0.4749 0.1>, {arriveTangent=1.714, interpMode=RCIM_Cubic, leaveTangent=1.714, time=0.7455, value=1}, <at 1 1 RCIM_Cubic>]}
       GreenCurve {keys [<at 0.4749 0.1>, {arriveTangent=1.714, interpMode=RCIM_Cubic, leaveTangent=1.714, time=0.7455, value=1}, <at 1 1 RCIM_Cubic>]}
       RedCurve {keys [<at 0.4749 1>,<at 0.7455 1 RCIM_Cubic>,<at 1 1 RCIM_Cubic>]}
      }})
    Scale Alpha=&RandomRangeFloat(Minimum=0.5, Maximum=0.65)
   )
  ]
  rendererClasses [&NiagaraSpriteRendererProperties]
  renderers [{rendererClass=&NiagaraSpriteRendererProperties}]
 }
 Side_Wind {
  emitterUpdateScript [
   &EmitterState()
   &SpawnBurst_Instantaneous(Spawn Count=3, Use Spawn Probability=false, UseLoopCountLimit=true)
   &SpawnBurst_Instantaneous(Spawn Count=3, Spawn Group=1, Use Spawn Probability=false, UseLoopCountLimit=true) {module=SpawnBurst_Instantaneous001}
  ]
  particleSpawnScript [
   &InitializeParticle(
    Color=<rgba 1 0.3739 0.3739 0.4>
    Lifetime Max=1.8
    Lifetime Min=2
    Lifetime Mode=::NewEnumerator1
    Mass Mode=::NewEnumerator3
    Mesh Scale=<5 5 7>
    Mesh Scale Mode=::NewEnumerator4
    Sprite Rotation Mode=::NewEnumerator2
    Sprite Size Mode=::NewEnumerator3
    Uniform Sprite Size=750
    UsePositionOffset=false
    Write Material Random=true
   )
   /SetVariables(Particles.Spread=0.0325) {isSetParameters=true, module=SetVariables_#1}
   &PartitionParticles(
    Particles Per Partition=3
    Partition Count Mode:ENiagara_InfiniteLoopDuration=::NewEnumerator1
    Partition Mode:ENiagara_PartitionMode_v2=::NewEnumerator1
    Write Partition Index to Ribbon Link Order:ENiagara_AttributeSamplingApplyOutput=::NewEnumerator0
    Write Partition to Ribbon ID=::NewEnumerator0
   )
   &ShapeLocation(
    MaskSpawnGroup=false
    Offset=<0 0 -275>
    Offset Mode=::NewEnumerator0
    Override Seed=false
    Ring / Disc Distribution Mode:ENiagaraSphereDistributionMode=::NewEnumerator1
    Ring Radius=3000
    Shape Primitive=::NewEnumerator4
    U Position=&Add_Float(
     A=0.25
     B=&MaskFloatBySpawnGroup(
      Masked Float=&Add_Float(A=$Particles.PartitionParticles.Normalized Index In Partition, B=$Particles.Spread)
      Passthrough Float=&Subtract_Float(A=$Particles.PartitionParticles.Normalized Index In Partition, B=$Particles.Spread)
     )
    )
    UseOwnerScale=true
   )
   &InitialMeshOrientation(
    Mesh Orientation Mode=::NewEnumerator2
    Rotation=&MaskVectorBySpawnGroup(
     Masked Vector=&MakeVector(X=0, Y=0, Z=&Add_Float(A=0.055, B=$Particles.PartitionParticles.Normalized Index In Partition))
     Passthrough Vector=&MakeVector(X=0, Y=0, Z=&Add_Float(A=-0.055, B=$Particles.PartitionParticles.Normalized Index In Partition))
    )
    Rotation Coordinate Space=::NewEnumerator0
    Use Rotation Vector=true
   )
   /SetVariables(Particles.MeshOrientation=&MultiplyQuaternion(
     Quaternion A=&MakeQuaternion(
      Quaternion From=0
      Axis=&ReturnMeshOrientationAxis(Mesh Orientation Axis=::XAxis, Mesh Orientation Quat=$Particles.MeshOrientation)
      Angle=47
      From=?
      To=?
      Angle Type=::NewEnumerator0
      Normalize Vectors=false
     )
     Quaternion B=$Particles.MeshOrientation
    )) {isSetParameters=true, module=SetVariables_#2}
   /SetVariables(Particles.Tube Axis:Vector3f=&ReturnMeshOrientationAxis(Mesh Orientation Axis=::YAxis, Mesh Orientation Quat=$Particles.MeshOrientation)) {isSetParameters=true, module=SetVariables_#3}
   /SetVariables(Particles.Linear Color=<rgba 1 0.2 0.2 1>) {isSetParameters=true}
  ]
  particleUpdateScript [
   &ParticleState(Kill Particles When Lifetime Has Elapsed=false, Let Infinitely Lived Particles Die When Emitter Deactivates=true)
   &ScaleMeshSize(Scale Factor=&MakeVector(
     X=1
     Y=1
     Z=&Waveform(
      "Waveform [1]"=::NewEnumerator12
      "[1] Input Position"=&RandomRangeFloat(Minimum=0, Maximum=10000)
      "[1] Phase Scale"=4
      "[1] Frequency"=1
      "[1] Amplitude Min/Max"=<0.7 1>
     )
    ))
   &DynamicMaterialParameters(
    Index 1 Param 1=&Add_Float(
     A=&Multiply_Float(A=$Emitter.Age, B=0.35)
     B=&RandomRangeFloat(Minimum=0, Maximum=100)
    )
    Index 1 Param 2=1
    Module.Index1Float0Write=true
    Module.Index1Float1Write=false
    Module.Index1Float2Write=false
    Module.Index1Float3Write=false
    Write Parameter Index 0=::NewEnumerator0
    Write Parameter Index 1=::NewEnumerator1
   ) {module=DynamicMaterialParameters001}
  ]
  rendererClasses [&NiagaraMeshRendererProperties]
  renderers [
   {rendererClass=&NiagaraMeshRendererProperties, rendererIndex=0}
   {rendererClass=&NiagaraMeshRendererProperties, rendererIndex=1}
  ]
 }
}
summary {
 systemName=NS_SecretBaseVentWind
 emitters [
  {emitterName=Heat, rendererClasses [&NiagaraSpriteRendererProperties]}
  {emitterName=Side_Wind, rendererClasses [&NiagaraMeshRendererProperties]}
  {emitterName=Motes, simTarget=GPUComputeSim, rendererClasses [&NiagaraSpriteRendererProperties]}
 ]
 userVariables []
}
